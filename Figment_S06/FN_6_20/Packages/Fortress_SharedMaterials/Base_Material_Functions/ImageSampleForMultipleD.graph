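# Material function that optionally reduces a four-component texture sample to a single value, for use inside other materials.
# A static switch on Use_dot returns either the dot product of Texture and Dot_Value or the Texture input unchanged.
# Dot_Value acts as a channel mask or weighting, with a preview default that selects red and alpha.
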MaterialFunction /Figment_S06/FN_6_20/Packages/Fortress_SharedMaterials/Base_Material_Functions/ImageSampleForMultipleD
graph {nodes [
  n1 = FunctionInput() {bUsePreviewValueAsDefault=true, inputName=Texture, inputType=FunctionInput_Vector4}
  n2 = FunctionInput() {bUsePreviewValueAsDefault=true, inputName=Dot_Value, inputType=FunctionInput_Vector4, sortPriority=1, previewValue=<1 0 0 1>}
  n3 = FunctionInput(Preview=StaticBool()) {bUsePreviewValueAsDefault=true, inputName=Use_dot, inputType=FunctionInput_StaticBool, sortPriority=2, previewValue=<1 0 0 1>}
  FunctionOutput(None=StaticSwitch(True=DotProduct(A=@n1, B=@n2), False=@n1, Value=@n3))
 ]}
