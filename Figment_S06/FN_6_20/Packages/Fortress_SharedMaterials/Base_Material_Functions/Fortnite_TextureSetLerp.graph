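# Fortnite material function that picks between diffuse, normal and spec-mask sets using TransitionInfo, defaulting to UV channel 1's red component.
# Up to ten diffuse inputs, each with an optional Dot value, feed LerpTexturesBasedOnInt_Float4.
# ReplaceTexturesBasedOnInt selects among six normals and six spec masks via static switches, cut to three on ES2/ES3_1.
# Spec mask RGB also outputs as Specular, Metalness and Roughness; the local Floor/Clamp chain is unconnected.
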MaterialFunction /Figment_S06/FN_6_20/Packages/Fortress_SharedMaterials/Base_Material_Functions/Fortnite_TextureSetLerp
graph {
 properties {
  description="Lerp between different textures based on the red channel in uv channel 1 (3ds Max uv channel 2)"
  libraryCategoriesText [Misc,FortNite]
 }
 nodes [
  n1 = FunctionInput(Preview=ComponentMask(None=TextureCoordinate() {coordinateIndex=1}) {r=true}) {bUsePreviewValueAsDefault=true, inputName=TransitionInfo, inputType=FunctionInput_Scalar, sortPriority=21}
  n2 = Floor(None=@n1)
  Clamp(None=Subtract(A=@n2, B=Constant() {r=5}))
  Clamp(None=@n2)
  Clamp(None=Subtract(A=@n2, B=Constant() {r=1}))
  Clamp(None=Subtract(A=@n2, B=Constant() {r=2}))
  Clamp(None=Subtract(A=@n2, B=Constant() {r=6}))
  Clamp(None=Subtract(A=@n2, B=Constant() {r=3}))
  Clamp(None=Subtract(A=@n2, B=Constant() {r=4}))
  Clamp(None=Subtract(A=@n2, B=Constant() {r=7}))
  Clamp(None=Subtract(A=@n2, B=Constant() {r=8}))
  FunctionInput() {bUsePreviewValueAsDefault=true, inputName=--------------, inputType=FunctionInput_Scalar, sortPriority=10}
  FunctionInput() {bUsePreviewValueAsDefault=true, inputName="-------------- ", inputType=FunctionInput_Scalar, sortPriority=20}
  n3 = FunctionInput() {bUsePreviewValueAsDefault=true, inputName=Normal1, previewValue=<0 0 1 1>}
  n4 = FunctionInput() {bUsePreviewValueAsDefault=true, inputName=Normal2, sortPriority=1, previewValue=<0 0 1 1>}
  n5 = FunctionInput() {bUsePreviewValueAsDefault=true, inputName=Normal3, sortPriority=2, previewValue=<0 0 1 1>}
  n6 = StaticBoolParameter() {group=Material_1, parameterName=Isolate Normal 1 To First Material Slot}
  n7 = StaticBool() {value=true}
  n8 = StaticBoolParameter() {group=Material_2, parameterName=Use Normal Texture 2}
  n9 = StaticBoolParameter() {group=Material_3, parameterName=Use Normal Texture 3}
  n10 = MaterialFunctionCall!(
   transitionFactor=@n1
   Default Color=@n3
   1=@n3
   2=@n4
   3=@n5
   4=FunctionInput() {bUsePreviewValueAsDefault=true, inputName=Normal4, sortPriority=3, previewValue=<0 0 1 1>}
   5=FunctionInput() {bUsePreviewValueAsDefault=true, inputName=Normal5, sortPriority=4, previewValue=<0 0 1 1>}
   6=FunctionInput() {bUsePreviewValueAsDefault=true, inputName=Normal6, sortPriority=5, previewValue=<0 0 1 1>}
   MaskFirstMaterial=@n6
   Use 1 Textures=@n7
   Use 2 Textures=@n8
   Use 3 Textures=@n9
   Use 4 Textures=StaticBoolParameter() {group=Material_4, parameterName=Use Normal Texture 4}
   Use 5 Textures=StaticBoolParameter() {group=Material_5, parameterName=Use Normal Texture 5}
   Use 6 Textures=StaticBoolParameter() {group=Material_6, parameterName=Use Normal Texture 6}
  ) {materialFunction=&/ReplaceTexturesBasedOnInt}
  n11 = Constant3Vector() {constant=<rgba 0 0 1 0>}
  n12 = StaticBool()
  n13 = MaterialFunctionCall(
   transitionFactor=@n1
   Default Color=@n3
   1=@n3
   2=@n4
   3=@n5
   4=@n11
   5=@n11
   6=@n11
   MaskFirstMaterial=@n6
   Use 1 Textures=@n7
   Use 2 Textures=@n8
   Use 3 Textures=@n9
   Use 4 Textures=@n12
   Use 5 Textures=@n12
   Use 6 Textures=@n12
  ) {materialFunction=&ReplaceTexturesBasedOnInt}
  FunctionOutput(None=FeatureLevelSwitch(
    Default=@n10.Final Output
    ES2=@n13.Final Output
    ES3_1=@n13.Final Output
    SM4_REMOVED=@n10.Final Output
    SM5=@n10.Final Output
    SM6=@n10.Final Output
   )) {outputName=Normals, sortPriority=3}
  n14 = Constant3Vector() {constant=<rgba 0.5 0 1 0>}
  n15 = FunctionInput() {bUsePreviewValueAsDefault=true, inputName=Spec1, previewValue=<1 0 0.2667 1>}
  n16 = FunctionInput() {bUsePreviewValueAsDefault=true, inputName=Spec2, sortPriority=1, previewValue=<1 0 0.2667 1>}
  n17 = FunctionInput() {bUsePreviewValueAsDefault=true, inputName=Spec3, sortPriority=2, previewValue=<1 0 0.2667 1>}
  n18 = StaticBoolParameter() {group=Material_1, parameterName=Isolate Spec Mask 1 To First Material Slot}
  n19 = StaticBoolParameter() {defaultValue=true, group=Material_1, parameterName=Use SpecMask 1}
  n20 = StaticBoolParameter() {group=Material_2, parameterName=Use SpecMask 2}
  n21 = StaticBoolParameter() {group=Material_3, parameterName=Use SpecMask 3}
  n22 = MaterialFunctionCall(
   transitionFactor=@n1
   Default Color=@n14
   1=@n15
   2=@n16
   3=@n17
   4=FunctionInput() {bUsePreviewValueAsDefault=true, inputName=Spec4, sortPriority=3, previewValue=<1 0 0.2667 1>}
   5=FunctionInput() {bUsePreviewValueAsDefault=true, inputName=Spec5, sortPriority=4, previewValue=<1 0 0.2667 1>}
   6=FunctionInput() {bUsePreviewValueAsDefault=true, inputName=Spec6, sortPriority=5, previewValue=<1 0 0.2667 1>}
   MaskFirstMaterial=@n18
   Use 1 Textures=@n19
   Use 2 Textures=@n20
   Use 3 Textures=@n21
   Use 4 Textures=StaticBoolParameter() {group=Material_4, parameterName=Use SpecMask 4}
   Use 5 Textures=StaticBoolParameter() {group=Material_5, parameterName=Use SpecMask 5}
   Use 6 Textures=StaticBoolParameter() {group=Material_6, parameterName=Use SpecMask 6}
  ) {materialFunction=&ReplaceTexturesBasedOnInt}
  n23 = StaticBool()
  n24 = MaterialFunctionCall(
   transitionFactor=@n1
   Default Color=@n14
   1=@n15
   2=@n16
   3=@n17
   4=@n14
   5=@n14
   6=@n14
   MaskFirstMaterial=@n18
   Use 1 Textures=@n19
   Use 2 Textures=@n20
   Use 3 Textures=@n21
   Use 4 Textures=@n23
   Use 5 Textures=@n23
   Use 6 Textures=@n23
  ) {materialFunction=&ReplaceTexturesBasedOnInt}
  n25 = FeatureLevelSwitch(
   Default=@n22.Final Output
   ES2=@n24.Final Output
   ES3_1=@n24.Final Output
   SM4_REMOVED=@n22.Final Output
   SM5=@n22.Final Output
   SM6=@n22.Final Output
  )
  FunctionOutput(None=@n25) {outputName=SpecularMasks, sortPriority=1}
  n26 = Constant4Vector() {constant=<rgba 0.46 0.4009 0.2599 0>}
  n27 = FunctionInput(Preview=@n26.RGBA) {bUsePreviewValueAsDefault=true, inputName=Texture1, inputType=FunctionInput_Vector4}
  n28 = FunctionInput() {bUsePreviewValueAsDefault=true, inputName=Dot_Texture1, inputType=FunctionInput_Vector4, sortPriority=21, previewValue=<1 0 0 1>}
  n29 = StaticBool()
  n30 = FunctionInput(Preview=@n29) {bUsePreviewValueAsDefault=true, inputName=Use_Dot_Texture1, inputType=FunctionInput_StaticBool, sortPriority=31, previewValue=<1 0 0 1>}
  n31 = FunctionInput(Preview=@n26.RGBA) {bUsePreviewValueAsDefault=true, inputName=Texture2, inputType=FunctionInput_Vector4, sortPriority=1}
  n32 = FunctionInput() {bUsePreviewValueAsDefault=true, inputName=Dot_Texture2, inputType=FunctionInput_Vector4, sortPriority=22, previewValue=<1 0 0 1>}
  n33 = FunctionInput(Preview=@n29) {bUsePreviewValueAsDefault=true, inputName=Use_Dot_Texture2, inputType=FunctionInput_StaticBool, sortPriority=32, previewValue=<1 0 0 1>}
  n34 = FunctionInput(Preview=@n26.RGBA) {bUsePreviewValueAsDefault=true, inputName=Texture3, inputType=FunctionInput_Vector4, sortPriority=2}
  n35 = FunctionInput() {bUsePreviewValueAsDefault=true, inputName=Dot_Texture3, inputType=FunctionInput_Vector4, sortPriority=23, previewValue=<1 0 0 1>}
  n36 = FunctionInput(Preview=@n29) {bUsePreviewValueAsDefault=true, inputName=Use_Dot_Texture3, inputType=FunctionInput_StaticBool, sortPriority=33, previewValue=<1 0 0 1>}
  n37 = FunctionInput(Preview=@n26.RGBA) {bUsePreviewValueAsDefault=true, inputName=Texture4, inputType=FunctionInput_Vector4, sortPriority=3}
  n38 = FunctionInput() {bUsePreviewValueAsDefault=true, inputName=Dot_Texture4, inputType=FunctionInput_Vector4, sortPriority=24, previewValue=<1 0 0 1>}
  n39 = FunctionInput(Preview=@n29) {bUsePreviewValueAsDefault=true, inputName=Use_Dot_Texture4, inputType=FunctionInput_StaticBool, sortPriority=34, previewValue=<1 0 0 1>}
  n40 = FunctionInput(Preview=@n26.RGBA) {bUsePreviewValueAsDefault=true, inputName=Texture5, inputType=FunctionInput_Vector4, sortPriority=4}
  n41 = FunctionInput() {bUsePreviewValueAsDefault=true, inputName=Dot_Texture5, inputType=FunctionInput_Vector4, sortPriority=25, previewValue=<1 0 0 1>}
  n42 = FunctionInput(Preview=@n29) {bUsePreviewValueAsDefault=true, inputName=Use_Dot_Texture5, inputType=FunctionInput_StaticBool, sortPriority=35, previewValue=<1 0 0 1>}
  n43 = FunctionInput(Preview=@n26.RGBA) {bUsePreviewValueAsDefault=true, inputName=Texture6, inputType=FunctionInput_Vector4, sortPriority=5}
  n44 = FunctionInput() {bUsePreviewValueAsDefault=true, inputName=Dot_Texture6, inputType=FunctionInput_Vector4, sortPriority=26, previewValue=<1 0 0 1>}
  n45 = FunctionInput(Preview=@n29) {bUsePreviewValueAsDefault=true, inputName=Use_Dot_Texture6, inputType=FunctionInput_StaticBool, sortPriority=36, previewValue=<1 0 0 1>}
  n46 = FunctionInput(Preview=@n26.RGBA) {bUsePreviewValueAsDefault=true, inputName=Texture7, inputType=FunctionInput_Vector4, sortPriority=6}
  n47 = FunctionInput() {bUsePreviewValueAsDefault=true, inputName=Dot_Texture7, inputType=FunctionInput_Vector4, sortPriority=27, previewValue=<1 0 0 1>}
  n48 = FunctionInput(Preview=@n29) {bUsePreviewValueAsDefault=true, inputName=Use_Dot_Texture7, inputType=FunctionInput_StaticBool, sortPriority=37, previewValue=<1 0 0 1>}
  n49 = FunctionInput(Preview=@n26.RGBA) {bUsePreviewValueAsDefault=true, inputName=Texture8, inputType=FunctionInput_Vector4, sortPriority=7}
  n50 = FunctionInput() {bUsePreviewValueAsDefault=true, inputName=Dot_Texture8, inputType=FunctionInput_Vector4, sortPriority=27, previewValue=<1 0 0 1>}
  n51 = FunctionInput(Preview=@n29) {bUsePreviewValueAsDefault=true, inputName=Use_Dot_Texture8, inputType=FunctionInput_StaticBool, sortPriority=38, previewValue=<1 0 0 1>}
  n52 = FunctionInput(Preview=@n26.RGBA) {bUsePreviewValueAsDefault=true, inputName=Texture9, inputType=FunctionInput_Vector4, sortPriority=8}
  n53 = FunctionInput() {bUsePreviewValueAsDefault=true, inputName=Dot_Texture9, inputType=FunctionInput_Vector4, sortPriority=27, previewValue=<1 0 0 1>}
  n54 = FunctionInput(Preview=@n29) {bUsePreviewValueAsDefault=true, inputName=Use_Dot_Texture9, inputType=FunctionInput_StaticBool, sortPriority=9, previewValue=<1 0 0 1>}
  n55 = FunctionInput(Preview=@n26.RGBA) {bUsePreviewValueAsDefault=true, inputName=Texture10, inputType=FunctionInput_Vector4, sortPriority=9}
  n56 = FunctionInput() {bUsePreviewValueAsDefault=true, inputName=Dot_Texture10, inputType=FunctionInput_Vector4, sortPriority=27, previewValue=<1 0 0 1>}
  n57 = FunctionInput(Preview=@n29) {bUsePreviewValueAsDefault=true, inputName=Use_Dot_Texture10, inputType=FunctionInput_StaticBool, sortPriority=40, previewValue=<1 0 0 1>}
  n58 = StaticBool()
  FunctionOutput(None=MaterialFunctionCall!(
    transitionFactor=@n1
    1=MaterialFunctionCall!(Texture=@n27, Dot_Value=@n28, Use_dot=@n30) {materialFunction=&/ImageSampleForMultipleD}.Result
    2=MaterialFunctionCall(Texture=@n31, Dot_Value=@n32, Use_dot=@n33) {materialFunction=&ImageSampleForMultipleD}.Result
    3=MaterialFunctionCall(Texture=@n34, Dot_Value=@n35, Use_dot=@n36) {materialFunction=&ImageSampleForMultipleD}.Result
    4=MaterialFunctionCall(Texture=@n37, Dot_Value=@n38, Use_dot=@n39) {materialFunction=&ImageSampleForMultipleD}.Result
    5=MaterialFunctionCall(Texture=@n40, Dot_Value=@n41, Use_dot=@n42) {materialFunction=&ImageSampleForMultipleD}.Result
    6=MaterialFunctionCall(Texture=@n43, Dot_Value=@n44, Use_dot=@n45) {materialFunction=&ImageSampleForMultipleD}.Result
    7=MaterialFunctionCall(Texture=@n46, Dot_Value=@n47, Use_dot=@n48) {materialFunction=&ImageSampleForMultipleD}.Result
    8=MaterialFunctionCall(Texture=@n49, Dot_Value=@n50, Use_dot=@n51) {materialFunction=&ImageSampleForMultipleD}.Result
    9=MaterialFunctionCall(Texture=@n52, Dot_Value=@n53, Use_dot=@n54) {materialFunction=&ImageSampleForMultipleD}.Result
    10=MaterialFunctionCall(Texture=@n55, Dot_Value=@n56, Use_dot=@n57) {materialFunction=&ImageSampleForMultipleD}.Result
    Use 2 Textures=FunctionInput(Preview=@n58) {bUsePreviewValueAsDefault=true, inputName=Use 2 Diffuse Textures, inputType=FunctionInput_StaticBool, sortPriority=10}
    Use 3 Textures=FunctionInput(Preview=@n58) {bUsePreviewValueAsDefault=true, inputName=Use 3 Diffuse Textures, inputType=FunctionInput_StaticBool, sortPriority=11}
    Use 4 Textures=FunctionInput(Preview=@n58) {bUsePreviewValueAsDefault=true, inputName=Use 4 Diffuse Textures, inputType=FunctionInput_StaticBool, sortPriority=12}
    Use 5 Textures=FunctionInput(Preview=@n58) {bUsePreviewValueAsDefault=true, inputName=Use 5 Diffuse Textures, inputType=FunctionInput_StaticBool, sortPriority=13}
    Use 6 Textures=FunctionInput(Preview=@n58) {bUsePreviewValueAsDefault=true, inputName=Use 6 Diffuse Textures, inputType=FunctionInput_StaticBool, sortPriority=14}
    Use 7 Textures=FunctionInput(Preview=@n58) {bUsePreviewValueAsDefault=true, inputName=Use 7 Diffuse Textures, inputType=FunctionInput_StaticBool, sortPriority=15}
    Use 8 Textures=FunctionInput(Preview=@n58) {bUsePreviewValueAsDefault=true, inputName=Use 8 Diffuse Textures, inputType=FunctionInput_StaticBool, sortPriority=16}
    Use 9 Textures=FunctionInput(Preview=@n58) {bUsePreviewValueAsDefault=true, inputName=Use 9 Diffuse Textures, inputType=FunctionInput_StaticBool, sortPriority=17}
    Use 10 Textures=FunctionInput(Preview=@n58) {bUsePreviewValueAsDefault=true, inputName=Use 10 Diffuse Textures, inputType=FunctionInput_StaticBool, sortPriority=18}
   ) {materialFunction=&/LerpTexturesBasedOnInt_Float4}.Final Output) {outputName=Diffuse}
  FunctionOutput(None=ComponentMask(None=@n25) {r=true}) {outputName=Specular, sortPriority=1}
  FunctionOutput(None=ComponentMask(None=@n25) {g=true}) {outputName=Metalness, sortPriority=1}
  FunctionOutput(None=ComponentMask(None=@n25) {b=true}) {outputName=Roughness, sortPriority=1}
 ]
}
